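# Material attributes function that adds optional dents and displacement to lake structure materials around a damage center offset from the actor position.
# A horizontal radial falloff masks world-aligned swirl noise blended with a tiled car-damage normal over the base normal.
# The same falloff scales noise-varied World Position Offset along the direction from the center.
# DamageControl from MPC_LakeStructure drives both effects, while IsStaticDamage keeps full normal damage without added displacement.
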MaterialFunction /Figment_S10/Environments/Sets/DigSite/Materials/MF_LakeStructure_AlteredDynamic
graph {nodes [
  n1 = FunctionInput() {inputType=FunctionInput_MaterialAttributes}
  n2 = GetMaterialAttributes(None=@n1) {attributeGetTypes [#1, #2]}
  n3 = ScalarParameter() {defaultValue=8788.1025, parameterName=SizeDamage}
  n4 = MaterialFunctionCall(
   TextureObject=TextureObject() {samplerType=SAMPLERTYPE_Normal, texture=/Figment_S10/Environments/Foliage/Textures/T_SwirlNoise_N}
   TextureSize=@n3
  ) {materialFunction=&WorldAlignedNormal}
  n5 = TextureCoordinate()
  n6 = ScalarParameter() {defaultValue=0.2, parameterName=DamageTiling}
  n7 = MaterialFunctionCall(
   Normal=TextureSample(UVs=Multiply(A=@n5, B=@n6)) {samplerType=SAMPLERTYPE_Normal, texture=/Figment_S10/Vehicles/FORT_Vehicles_01/CarDamageNormal}.RGB
   Flatness=ScalarParameter() {defaultValue=-1, parameterName=NoiseDamageFlatten}
  ) {materialFunction=&FlattenNormal}
  n8 = Constant() {r=1}
  n9 = ScalarParameter() {defaultValue=0.5333, parameterName=NormalDamage}
  n10 = WorldPosition()
  n11 = Add(A=ActorPositionWS(), B=VectorParameter() {parameterName=DamageCenter, defaultValue=<rgba 0 0 4000 1>}.RGB)
  n12 = Distance(A=ComponentMask(None=@n10.XYZ) {g=true, r=true}, B=ComponentMask(None=@n11) {g=true, r=true})
  n13 = ScalarParameter() {defaultValue=2104.5234, parameterName=DamageRadius}
  n14 = OneMinus(None=Saturate(None=Divide(A=@n12, B=@n13)))
  n15 = MaterialFunctionCall(
   BaseNormal=@n2.Normal
   AdditionalNormal=MaterialFunctionCall(
    Normal=MaterialFunctionCall(BaseNormal=@n4.Z Texture, AdditionalNormal=@n7.Result) {materialFunction=&BlendAngleCorrectedNormals}.Result
    Flatness=LinearInterpolate(A=@n8, B=@n9, Alpha=@n14)
   ) {materialFunction=&FlattenNormal}.Result
  ) {materialFunction=&BlendAngleCorrectedNormals}
  n16 = CollectionParameter() {collection=&/MPC_LakeStructure, parameterName=DamageControl}
  n17 = StaticSwitchParameter(True=@n15.Result, False=LinearInterpolate(A=@n2.Normal, B=@n15.Result, Alpha=@n16)) {parameterName=IsStaticDamage}
  n18 = Multiply(
   A=Normalize(VectorInput=Subtract(A=@n10.XYZ, B=@n11))
   B=Multiply(A=@n14, B=ScalarParameter() {defaultValue=0.7798, parameterName=BiasDamage})
  )
  n19 = MaterialFunctionCall(
   TextureObject=TextureObject() {samplerType=SAMPLERTYPE_LinearGrayscale, texture=/Figment_S10/Effects/Fort_Effects/BluGlow/Violentium/T_SwirlNoise_N_height}
   TextureSize=@n3
  ) {materialFunction=&WorldAlignedTexture}
  n20 = LinearInterpolate(
   A=Constant()
   B=LinearInterpolate(
    A=ScalarParameter() {defaultValue=-500, parameterName=DamagePush2}
    B=ScalarParameter() {defaultValue=1000, parameterName=DamagePush}
    Alpha=ComponentMask(None=@n19.Z Texture) {r=true}
   )
   Alpha=@n16
  )
  n21 = StaticSwitchParameter(True=@n2.WorldPositionOffset, False=Add(A=@n2.WorldPositionOffset, B=Multiply(A=@n18, B=@n20))) {parameterName=IsStaticDamage}
  FunctionOutput(None=StaticSwitchParameter(
    True=SetMaterialAttributes!(MaterialAttributes=@n2.MaterialAttributes, Normal=@n17, WorldPositionOffset=@n21) {attributeSetTypes [#1, #2]}
    False=@n1
   ) {parameterName=UseDamage})
  ScalarParameter() {defaultValue=1, parameterName=MasterDamageControl}
 ]}
